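# Animation Blueprint for the tutorial third-person character, driving its locomotion pose from movement speed.
# Each animation update gets the pawn owner, checks it is valid, casts it to TutorialCharacter and reads its CharacterMovement component.
# It sets Speed to velocity length divided by MaxWalkSpeed, a value relative to walk speed.
# The AnimGraph feeds Speed into the one-axis blendspace NewBlendSpace1D, whose result is the output pose.
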AnimBlueprint /Engine/Tutorial/SubEditors/TutorialAssets/Character/TutorialTPP_AnimBlueprint
target_skeleton=&/TutorialTPP_Skeleton
graphs [
 AnimGraph {nodes [Output Pose(Result=Blendspace Player 'NewBlendSpace1D'!(X=Get Speed() -> (Speed:double)) -> (Pose) {blendSpace=&/NewBlendSpace1D}) {layerGroup=None}]}
 EventGraph {nodes [
   Event BlueprintUpdateAnimation()
   n1 = TryGetPawnOwner()
   > Is Valid(InputObject=@n1)
   > n2 = Cast To TutorialCharacter!(execute, Object=@n1) -> (then, CastFailed, AsTutorial Character:Tutorial Character Object Reference) {is=title K2Node_DynamicCast}
   n3 = Get CharacterMovement!(self:Character Object Reference=@n2) -> (CharacterMovement:Character Movement Component Object Reference)
   n4 = Get Velocity!(self:Movement Component Object Reference=@n3) -> (Velocity:Vector)
   > Set Speed(Speed:double=float / float(A=Vector Length(A=@n4), B=Get MaxWalkSpeed!(self:Character Movement Component Object Reference=@n3) -> (MaxWalkSpeed:float)) {is=title K2Node_CallFunction}) -> (Output_Get:double)
  ]}
]
variables [Speed {type=number}]
