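# TutorialTPP_Mat is a subsurface-shaded material in the engine's tutorial character assets, apparently for the third-person tutorial character.
# It is flagged for use on skeletal meshes and with mesh deformers.
# Its graph holds only constants: an orange color, a red color and two scalars, with no textures or parameters.
# The file does not show which material inputs these constants feed.
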Material /Engine/Tutorial/SubEditors/TutorialAssets/Character/TutorialTPP_Mat
graph {
 properties {
  bUsedWithMeshDeformer=true
  bUsedWithSkeletalMesh=true
  pixelDepthOffsetMode=PDOM_Legacy
  shadingModel=MSM_Subsurface
  usedShadingModels=Subsurface
 }
 nodes [
  Constant3Vector() {constant=<rgba 0.89 0.4138 0 0>}
  Constant3Vector() {constant=<rgba 1 0 0 0>}
  Constant()
  Constant() {r=1}
 ]
}
