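# Physics asset for the engine tutorial's TPP character, giving its humanoid skeleton collision and ragdoll-style bodies.
# Each major bone from Root and pelvis through spine, head, arms, legs and toe balls carries a single capsule, with no finger bodies.
# Constraints link each body to its parent up to Root, all sharing identical swing and twist limit angles rather than per-joint anatomical tuning.
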PhysicsAsset /Engine/Tutorial/SubEditors/TutorialAssets/Character/TutorialTPP_PhysicsAsset
bodies {
 Root [{shapeType=Capsule, length=15.15, radius=15.15, rotation=<rot 0.0501 -90 0>}]
 ball_l [{shapeType=Capsule, length=6.282, radius=3.0682, center=<6.578 -0.7637 1.8635>, rotation=<rot -2.5931 -90 23.5736>}]
 ball_r [{shapeType=Capsule, length=6.282, radius=3.0682, center=<-6.5786 0.7638 -1.8587>, rotation=<rot -2.5931 -90 23.5736>}]
 calf_l [{shapeType=Capsule, length=24.8647, radius=7.6185, center=<-21.7481 -1.4897 -0.2396>, rotation=<rot 0 -89.9372 -89.9785>}]
 calf_r [{shapeType=Capsule, length=24.8647, radius=7.6185, center=<21.7479 1.4896 0.2445>, rotation=<rot 0 90.0628 -90.0214>}]
 clavicle_l [{shapeType=Capsule, length=12.1394, radius=10.7527, center=<15.6625 1.8617 -0.0264>, rotation=<rot 0 91.2659 -89.8182>}]
 clavicle_r [{shapeType=Capsule, length=12.1395, radius=10.7527, center=<-15.6668 -1.8611 0.0242>, rotation=<rot 0 -88.7341 -90.1817>}]
 foot_l [{shapeType=Capsule, length=9.7784, radius=8.4439, center=<-7.1684 3.0424 1.03>, rotation=<rot 0 -135.0793 -90.3377>}]
 foot_r [{shapeType=Capsule, length=9.7784, radius=8.4439, center=<7.1683 -3.043 -1.0252>, rotation=<rot 0 44.9206 -89.6623>}]
 hand_l [{shapeType=Capsule, length=5.8938, radius=4.7302, center=<7.3357 0.1727 0.2092>, rotation=<rot 0 -88.7637 -84.9311>}]
 hand_r [{shapeType=Capsule, length=5.8852, radius=4.7341, center=<-7.3473 -0.1741 -0.1815>, rotation=<rot 0 -88.8211 -85.5162>}]
 head [{shapeType=Capsule, length=13.1228, radius=10.8164, center=<5.8288 1.0245 0.0772>, rotation=<rot 0 -130.8558 -90>}]
 lowerarm_l [{shapeType=Capsule, length=15.9459, radius=6.4974, center=<19.1069 -0.3405 -0.1208>, rotation=<rot 0 89.7844 -89.989>}]
 lowerarm_r [{shapeType=Capsule, length=15.9459, radius=6.4974, center=<-19.1098 0.3406 0.1169>, rotation=<rot 0 -90.2156 -90.0112>}]
 neck_01 [{shapeType=Capsule, length=4.7385, radius=6.4887, center=<3.1108 -0.5319 0.0039>, rotation=<rot 0 90 -90>}]
 pelvis [{shapeType=Capsule, length=15.1979, radius=13.8824, center=<-0.8754 -0.1523 0.118>, rotation=<rot 1.1996 -90 3.9242>}]
 spine_01 [{shapeType=Capsule, length=4.4901, radius=12.5897, center=<2.7925 1.37 0.0024>, rotation=<rot 0 90 -90>}]
 spine_02 [{shapeType=Capsule, length=5.5659, radius=13.9085, center=<-1.7044 0.707 0.0024>, rotation=<rot 0 90 -90>}]
 spine_03 [{shapeType=Capsule, length=19.0337, radius=17.8696, center=<1.3114 -0.1537 -0.3406>, rotation=<rot 0 -103.0879 -84.9761>}]
 thigh_l [{shapeType=Capsule, length=32.8203, radius=9.8333, center=<-25.341 0.7374 0.8091>, rotation=<rot 0 -89.9912 -89.9619>}]
 thigh_r [{shapeType=Capsule, length=32.8203, radius=9.8333, center=<25.341 -0.7375 -0.8042>, rotation=<rot 0 90.0088 -90.0382>}]
 upperarm_l [{shapeType=Capsule, length=13.351, radius=6.6749, center=<16.2268 0.6464 0.4143>, rotation=<rot 0 90 -90>}]
 upperarm_r [{shapeType=Capsule, length=13.351, radius=6.6749, center=<-16.2297 -0.6467 -0.4182>, rotation=<rot 0 -90 -89.9999>}]
}
constraints [
 {bone1Name=pelvis, bone2Name=Root, swing1LimitDegrees=45, swing2LimitDegrees=45, twistLimitDegrees=45}
 {bone1Name=spine_01, bone2Name=pelvis, swing1LimitDegrees=45, swing2LimitDegrees=45, twistLimitDegrees=45}
 {bone1Name=spine_02, bone2Name=spine_01, swing1LimitDegrees=45, swing2LimitDegrees=45, twistLimitDegrees=45}
 {bone1Name=spine_03, bone2Name=spine_02, swing1LimitDegrees=45, swing2LimitDegrees=45, twistLimitDegrees=45}
 {bone1Name=clavicle_l, bone2Name=spine_03, swing1LimitDegrees=45, swing2LimitDegrees=45, twistLimitDegrees=45}
 {bone1Name=upperarm_l, bone2Name=clavicle_l, swing1LimitDegrees=45, swing2LimitDegrees=45, twistLimitDegrees=45}
 {bone1Name=lowerarm_l, bone2Name=upperarm_l, swing1LimitDegrees=45, swing2LimitDegrees=45, twistLimitDegrees=45}
 {bone1Name=hand_l, bone2Name=lowerarm_l, swing1LimitDegrees=45, swing2LimitDegrees=45, twistLimitDegrees=45}
 {bone1Name=clavicle_r, bone2Name=spine_03, swing1LimitDegrees=45, swing2LimitDegrees=45, twistLimitDegrees=45}
 {bone1Name=upperarm_r, bone2Name=clavicle_r, swing1LimitDegrees=45, swing2LimitDegrees=45, twistLimitDegrees=45}
 {bone1Name=lowerarm_r, bone2Name=upperarm_r, swing1LimitDegrees=45, swing2LimitDegrees=45, twistLimitDegrees=45}
 {bone1Name=hand_r, bone2Name=lowerarm_r, swing1LimitDegrees=45, swing2LimitDegrees=45, twistLimitDegrees=45}
 {bone1Name=neck_01, bone2Name=spine_03, swing1LimitDegrees=45, swing2LimitDegrees=45, twistLimitDegrees=45}
 {bone1Name=head, bone2Name=neck_01, swing1LimitDegrees=45, swing2LimitDegrees=45, twistLimitDegrees=45}
 {bone1Name=thigh_l, bone2Name=pelvis, swing1LimitDegrees=45, swing2LimitDegrees=45, twistLimitDegrees=45}
 {bone1Name=calf_l, bone2Name=thigh_l, swing1LimitDegrees=45, swing2LimitDegrees=45, twistLimitDegrees=45}
 {bone1Name=foot_l, bone2Name=calf_l, swing1LimitDegrees=45, swing2LimitDegrees=45, twistLimitDegrees=45}
 {bone1Name=ball_l, bone2Name=foot_l, swing1LimitDegrees=45, swing2LimitDegrees=45, twistLimitDegrees=45}
 {bone1Name=thigh_r, bone2Name=pelvis, swing1LimitDegrees=45, swing2LimitDegrees=45, twistLimitDegrees=45}
 {bone1Name=calf_r, bone2Name=thigh_r, swing1LimitDegrees=45, swing2LimitDegrees=45, twistLimitDegrees=45}
 {bone1Name=foot_r, bone2Name=calf_r, swing1LimitDegrees=45, swing2LimitDegrees=45, twistLimitDegrees=45}
 {bone1Name=ball_r, bone2Name=foot_r, swing1LimitDegrees=45, swing2LimitDegrees=45, twistLimitDegrees=45}
]
